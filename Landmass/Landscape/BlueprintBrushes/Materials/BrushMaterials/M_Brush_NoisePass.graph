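# Unlit Landmass landscape brush material that generates a procedural noise height pass and outputs it packed as RG8 height.
# Noise from tiled, seeded brush UVs is optionally inverted, remapped above a floor and scaled by NoiseHeight.
# The shape mask function supplies brush height, edge falloff and previous height as named reroutes.
# AdditiveBlend lerps both how noise is built and how heights combine, from smooth-max intersection to plain addition.
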Material /Landmass/Landscape/BlueprintBrushes/Materials/BrushMaterials/M_Brush_NoisePass
graph {
 properties {pixelDepthOffsetMode=PDOM_Legacy, shadingModel=MSM_Unlit, usedShadingModels=Unlit}
 nodes [
  n1 = NamedRerouteUsage()
  n2 = NamedRerouteUsage()
  n3 = Multiply(
   A=Multiply(A=ScalarParameter() {parameterName=IntersectionSmoothness}, B=ScalarParameter() {parameterName=NoiseHeight})
   B=NamedRerouteUsage()
  )
  n4 = NamedRerouteUsage()
  n5 = NamedRerouteUsage()
  MaterialFunctionCall!(Height Int -32k to 32k=LinearInterpolate(
    A=MaterialFunctionCall!(A=@n1, B=@n2, K=@n3) {materialFunction=/Landmass/Landscape/OldPrototype_BP/Mats/MF/SmoothIntersections}.Smooth Max
    B=Add(A=@n4, B=@n5)
    Alpha=ScalarParameter() {parameterName=AdditiveBlend}
   )) {materialFunction=/Landmass/Landscape/BlueprintBrushes/MF/PackRG8Height}
  n6 = MaterialFunctionCall!(Falloff Width Override, Falloff Offset Override) {materialFunction=/Landmass/Landscape/BlueprintBrushes/Materials/BrushMaterials/BasicFunctions/LandmassBrush_ShapeMask}
  NamedRerouteDeclaration(None=@n6.Brush Base Height) {name=Current Height, nodeColor=<rgba 0.1484 1 0 1>}
  NamedRerouteDeclaration(None=@n6.Brush Edge Falloff) {name=Brush Alpha, nodeColor=<rgba 1 0 0.1641 1>}
  n7 = MaterialFunctionCall() {materialFunction=/Landmass/Landscape/BlueprintBrushes/MF/LandmassBrush_UVs}
  n8 = ScalarParameter() {defaultValue=2, parameterName=NoiseTiling}
  n9 = MaterialFunctionCall!(
   Position=AppendVector(A=Multiply(A=@n7.UVs, B=@n8), B=ScalarParameter() {defaultValue=1, parameterName=NoiseSeed})
   Octaves=ScalarParameter() {defaultValue=4, parameterName=NoiseOctaves}
   Persistence=ScalarParameter() {defaultValue=0.35, parameterName=NoisePersistence}
   Turbulence=ScalarParameter() {parameterName=UseNoiseTurbulence}
  ) {materialFunction=/Landmass/Landscape/BlueprintBrushes/MF/LandmassBrush_Noise}
  n10 = OneMinus(None=@n9.Result)
  n11 = ScalarParameter() {parameterName=InvertNoise}
  n12 = ScalarParameter() {parameterName=NoiseFloor}
  n13 = Divide(A=Subtract(A=LinearInterpolate(A=@n9.Result, B=@n10, Alpha=@n11), B=@n12), B=OneMinus(None=@n12))
  n14 = Add(A=Multiply(A=@n13, B=ScalarParameter() {parameterName=NoiseHeight}), B=NamedRerouteUsage())
  n15 = NamedRerouteUsage()
  n16 = Multiply(
   A=MaterialFunctionCall(A=@n13, B=Constant(), K=ScalarParameter() {parameterName=IntersectionSmoothness}) {materialFunction=&SmoothIntersections}.Smooth Max
   B=ScalarParameter() {parameterName=NoiseHeight}
  )
  n17 = NamedRerouteUsage()
  NamedRerouteDeclaration(None=LinearInterpolate(A=Multiply(A=@n14, B=@n15), B=Multiply(A=@n16, B=@n17), Alpha=ScalarParameter() {parameterName=AdditiveBlend})) {name=Final Noise, nodeColor=<rgba 0.1719 1 0 1>}
  NamedRerouteDeclaration(None=@n6.Previous Height) {name=Previous Height, nodeColor=<rgba 0 0.4609 1 1>}
 ]
}
