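# Unlit Landmass landscape brush material that adds procedural noise to terrain height within a brush shape.
# LandmassBrush_ShapeMask supplies base height, edge falloff and previous height through named reroutes.
# Seeded, tiled brush UVs feed a multi-octave noise function, whose result is optionally inverted, remapped by a floor, smooth-clamped and scaled.
# BrushBlendModes receives alpha, smooth min, smooth max and additive variants, and the result is packed as RG8 height.
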Material /Landmass/Landscape/BlueprintBrushes/Materials/BrushMaterials/M_Brush_ProceduralNoise
graph {
 properties {pixelDepthOffsetMode=PDOM_Legacy, shadingModel=MSM_Unlit, usedShadingModels=Unlit}
 nodes [
  n1 = NamedRerouteUsage()
  n2 = NamedRerouteUsage()
  n3 = NamedRerouteUsage()
  n4 = NamedRerouteUsage()
  n5 = Add(A=NamedRerouteUsage(), B=NamedRerouteUsage())
  n6 = ScalarParameter() {parameterName=Smoothness}
  n7 = NamedRerouteUsage()
  n8 = Subtract(A=NamedRerouteUsage(), B=NamedRerouteUsage())
  n9 = ScalarParameter() {parameterName=Smoothness}
  n10 = Multiply(A=NamedRerouteUsage(), B=NamedRerouteUsage())
  n11 = NamedRerouteUsage()
  MaterialFunctionCall!(Height Int -32k to 32k=MaterialFunctionCall!(
    AlphaBlend=LinearInterpolate(A=@n1, B=@n2, Alpha=@n3)
    Min=MaterialFunctionCall!(A=@n4, B=@n5, K=@n6) {materialFunction=/Landmass/Landscape/BlueprintBrushes/MF/SmoothIntersections}.Smooth Min
    Max=MaterialFunctionCall(A=@n7, B=@n8, K=@n9) {materialFunction=&SmoothIntersections}.Smooth Max
    Additive=Add(A=@n10, B=@n11)
   ) {materialFunction=/Landmass/Landscape/BlueprintBrushes/MF/BrushBlendModes}.Result) {materialFunction=/Landmass/Landscape/BlueprintBrushes/MF/PackRG8Height}
  n12 = MaterialFunctionCall!(Falloff Width Override, Falloff Offset Override) {materialFunction=/Landmass/Landscape/BlueprintBrushes/Materials/BrushMaterials/BasicFunctions/LandmassBrush_ShapeMask}
  NamedRerouteDeclaration(None=@n12.Brush Base Height) {name=Current Height, nodeColor=<rgba 0.1484 1 0 1>}
  NamedRerouteDeclaration(None=@n12.Brush Edge Falloff) {name=Brush Edge Mask, nodeColor=<rgba 1 0 0.1641 1>}
  n13 = Multiply(
   A=MaterialFunctionCall(A=NamedRerouteUsage(), B=Constant(), K=ScalarParameter() {parameterName=IntersectionSmoothness}) {materialFunction=&SmoothIntersections}.Smooth Max
   B=ScalarParameter() {parameterName=NoiseHeight}
  )
  n14 = NamedRerouteUsage()
  NamedRerouteDeclaration(None=Add(A=NamedRerouteUsage(), B=Multiply(A=@n13, B=@n14))) {name=Brush Height Final, nodeColor=<rgba 1 0.9375 0 1>}
  n15 = MaterialFunctionCall() {materialFunction=/Landmass/Landscape/BlueprintBrushes/MF/LandmassBrush_UVs}
  n16 = ScalarParameter() {defaultValue=2, parameterName=NoiseTiling}
  n17 = MaterialFunctionCall!(
   Position=AppendVector(A=Multiply(A=@n15.UVs, B=@n16), B=ScalarParameter() {defaultValue=1, parameterName=NoiseSeed})
   Octaves=ScalarParameter() {defaultValue=4, parameterName=NoiseOctaves}
   Persistence=ScalarParameter() {defaultValue=0.35, parameterName=NoisePersistence}
   Turbulence=ScalarParameter() {parameterName=UseNoiseTurbulence}
  ) {materialFunction=/Landmass/Landscape/BlueprintBrushes/MF/LandmassBrush_Noise}
  n18 = LinearInterpolate(A=@n17.Result, B=OneMinus(None=@n17.Result), Alpha=ScalarParameter() {parameterName=InvertNoise})
  n19 = ScalarParameter() {parameterName=NoiseFloor}
  NamedRerouteDeclaration(None=Divide(A=Subtract(A=@n18, B=@n19), B=OneMinus(None=@n19))) {name=Noise Value 0-1, nodeColor=<rgba 0.094 0.3275 1 1>}
  n20 = ScalarParameter() {defaultValue=1, parameterName=FalloffRelativeHeight}
  n21 = ScalarParameter() {defaultValue=4096, parameterName=BrushSize}
  NamedRerouteDeclaration(None=LinearInterpolate(A=Multiply(A=@n20, B=@n21), Alpha=NamedRerouteUsage()) {constB=0}) {name=Edge Falloff Height, nodeColor=<rgba 0 0.5781 1 1>}
  NamedRerouteDeclaration(None=@n12.Previous Height) {name=Previous Height, nodeColor=<rgba 0 0.4609 1 1>}
 ]
}
