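# Material function for Landmass landscape brushes that masks terrain to a band between Min and Max Elevation.
# It point-samples the HeightRT texture at texel centers and unpacks the RG channels into elevation.
# Each limit becomes a saturated ramp over Falloff Width with additive Noise, with the upper ramp shifted one width lower.
# It outputs both masks and their minimum as Combined Masks.
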MaterialFunction /Landmass/Landscape/BlueprintBrushes/Materials/BrushMaterials/BasicFunctions/LandmassBrush_ElevationMask
graph {nodes [
  n1 = NamedRerouteUsage()
  n2 = Subtract(A=@n1, B=FunctionInput() {bUsePreviewValueAsDefault=true, inputName=Min Elevation, inputType=FunctionInput_Scalar, sortPriority=1})
  n3 = FunctionInput() {bUsePreviewValueAsDefault=true, inputName=Falloff Width, inputType=FunctionInput_Scalar, sortPriority=3, previewValue=<15 0 0 1>}
  n4 = FunctionInput() {bUsePreviewValueAsDefault=true, inputName=Noise, inputType=FunctionInput_Scalar, sortPriority=4}
  n5 = Saturate(None=Add(A=Divide(A=@n2, B=@n3), B=@n4))
  n6 = FunctionInput() {bUsePreviewValueAsDefault=true, inputName=Max Elevation, inputType=FunctionInput_Scalar, sortPriority=2, previewValue=<5000 0 0 1>}
  n7 = Subtract(A=Multiply(A=Subtract(A=@n1, B=@n6)) {constB=-1}, B=@n3)
  n8 = Saturate(None=Add(A=Divide(A=@n7, B=@n3), B=@n4))
  FunctionOutput(None=Min(
    A=FunctionOutput(None=@n5) {outputName=Min Elevation Mask, sortPriority=1}
    B=FunctionOutput(None=@n8) {outputName=Max Elevation Mask, sortPriority=2}
   )) {outputName=Combined Masks}
  n9 = TextureSampleParameter2D(UVs=MaterialFunctionCall() {materialFunction=/Landmass/Landscape/BlueprintBrushes/MF/PointSampleLandscapeUVs}.TexelCenterUVs) {parameterName=HeightRT, texture=/Engine/EngineResources/Black_Low.Black}
  NamedRerouteDeclaration(None=MaterialFunctionCall!(RG=ComponentMask(None=@n9.RGB) {g=true, r=true}) {materialFunction=/Landmass/Landscape/OldPrototype_BP/Mats/MF/UnpackRG8Height}.Result) {name=Elevation, nodeColor=<rgba 0.1953 1 0 1>}
  OneMinus(None=@n8)
 ]}
