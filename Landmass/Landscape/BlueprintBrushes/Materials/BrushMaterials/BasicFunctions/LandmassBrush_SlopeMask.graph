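# Landmass brush material function that outputs Min, Max and Combined slope masks from landscape height for brush materials.
# A Custom HLSL node decodes 16-bit height from the HeightRT red and green channels at the texel and its four axis neighbors.
# The largest difference, scaled by ZScale and GridSize, becomes a slope ratio compared against Min Slope, Max Slope and Slope Falloff inputs.
# The two saturated masks multiply into the combined output.
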MaterialFunction /Landmass/Landscape/BlueprintBrushes/Materials/BrushMaterials/BasicFunctions/LandmassBrush_SlopeMask
graph {nodes [
  MaterialExpressionTextureObjectParameter_0 = TextureObjectParameter() {parameterName=HeightRT, texture=&Black}
  MaterialExpressionMaterialFunctionCall_5 = MaterialFunctionCall() {materialFunction=/Landmass/Landscape/BlueprintBrushes/MF/PointSampleLandscapeUVs}
  MaterialExpressionAppendVector_0 = AppendVector(
   A=CollectionParameter() {collection=/Landmass/Landscape/BlueprintBrushes/MPC/MPC_Landscape, parameterName=RTResX}
   B=CollectionParameter() {collection=&MPC_Landscape, parameterName=RTResY}
  )
  MaterialExpressionAppendVector_3 = AppendVector(
   A=CollectionParameter() {collection=&MPC_Landscape, parameterName=LSQuadsX}
   B=CollectionParameter() {collection=&MPC_Landscape, parameterName=LSQuadsY}
  )
  n1 = Custom!(
   Tex=@MaterialExpressionTextureObjectParameter_0
   UV=@MaterialExpressionMaterialFunctionCall_5.TexelCenterUVs
   TextureSize=@MaterialExpressionAppendVector_0
   LandscapeResolution=@MaterialExpressionAppendVector_3
  ) {
   code="float2 texelsize = 1/TextureSize;\r\n\r\nfloat2 xoffsets[2] = {float2(1,0) , float2(-1,0)};\r\nfloat2 yoffsets[2] = {float2(0,1) , float2(0,-1)};\r\n\r\nfloat scalars[9] = {1, 1, 1, 1, 1,0.707, 0.707, 0.707, 0.707};\r\n\r\nfloat minz = 1000000;\r\nfloat maxz = 0;\r\nfloat accum = 0;\r\nfloat biggestdiff = 0;\r\n\r\nfloat2 centersample = Tex.SampleLevel(TexSampler, UV, 0);\r\nfloat centervalue = clamp(float(((int)round(centersample.r * 255.0) << 8) | (int)round(centersample.g * 255.0)), 0.0, 65536.0);\r\n\r\nfloat numsamples = 1;\r\nfloat2 curTexel = floor(UV * TextureSize);\r\n\r\nfor(int i = 0; i < 2; i++)\r\n{\r\n\tif(curTexel.x > 0 &&&& curTexel.x < (LandscapeResolution.x - 1))\r\n\t{\r\n\t\tfloat2 curUV = UV + (xoffsets[i] * texelsize);\r\n\t\tcurUV = clamp(curUV, 0.5*texelsize, 1 - (0.5*texelsize));\r\n\t\t\r\n\t\tfloat2 SourceSample = Tex.SampleLevel(TexSampler, curUV, 0);\r\n\t\tfloat offsetsample = clamp(float(((int)round(SourceSample.r * 255.0) << 8) | (int)round(SourceSample.g * 255.0)), 0.0, 65536.0);\r\n\t\t\r\n\t\t\r\n\t\tminz = min(minz, offsetsample.x);\r\n\t\tmaxz = max(maxz, offsetsample.x);\r\n\t\tbiggestdiff = max(abs(maxz - minz) * scalars[i], biggestdiff);\r\n\t\t\r\n\t\taccum += offsetsample;\r\n\t\tnumsamples += 1.0;\r\n\t}\r\n}\r\n\r\nfor(int j = 0; j < 2; j++)\r\n{\r\n\t\r\n\tif(curTexel.y > 0 &&&& curTexel.y < (LandscapeResolution.y - 1))\r\n\t{\r\n\t\tfloat2 curUV = UV + (yoffsets[j] * texelsize);\r\n\t\tcurUV = clamp(curUV, 0.5*texelsize, 1 - (0.5*texelsize));\r\n\t\t\r\n\t\tfloat2 SourceSample = Tex.SampleLevel(TexSampler, curUV, 0);\r\n\t\tfloat offsetsample = clamp(float(((int)round(SourceSample.r * 255.0) << 8) | (int)round(SourceSample.g * 255.0)), 0.0, 65536.0);\r\n\t\t\r\n\t\t\r\n\t\tminz = min(minz, offsetsample.x);\r\n\t\tmaxz = max(maxz, offsetsample.x);\r\n\t\tbiggestdiff = max(abs(maxz - minz) * scalars[j], biggestdiff);\r\n\t\t\r\n\t\taccum += offsetsample;\r\n\t\tnumsamples += 1.0;\r\n\t}\r\n}\r\n\r\n\r\naccum += centervalue;\r\naccum /= numsamples;\r\n\r\nreturn float3(biggestdiff, maxz, accum );"
   description=Sample Neighbors
   inputs [
    {
     inputName=Tex
     input {expression=&self:MaterialExpressionTextureObjectParameter_0}
    }
    {
     inputName=UV
     input {expression=&self:MaterialExpressionMaterialFunctionCall_5}
    }
    {
     inputName=TextureSize
     input {expression=&self:MaterialExpressionAppendVector_0}
    }
    {
     inputName=LandscapeResolution
     input {expression=&self:MaterialExpressionAppendVector_3}
    }
   ]
  }
  n2 = MaterialFunctionCall(Float3=@n1) {
   desc="R=Biggest Diff, G=maxZ of 4 neighbors, B=Average of center and 4 neighbors"
   materialFunction=&BreakOutFloat3Components
  }
  n3 = Divide(
   A=Multiply(A=@n2.R, B=ScalarParameter() {defaultValue=1.28, parameterName=ZScale})
   B=Multiply(A=ScalarParameter() {defaultValue=128, parameterName=GridSize}) {constB=2}
  )
  n4 = Subtract(A=@n3, B=Subtract(A=NamedRerouteUsage(), B=NamedRerouteUsage()))
  n5 = NamedRerouteUsage()
  n6 = OneMinus(None=Saturate(None=Divide(A=@n4, B=@n5)))
  n7 = NamedRerouteUsage()
  n8 = NamedRerouteUsage()
  n9 = Divide(A=Add(A=Multiply(A=@n3) {constB=-1}, B=Add(A=@n7, B=@n8)), B=NamedRerouteUsage())
  n10 = Constant() {r=1}
  n11 = If(A=NamedRerouteUsage(), "A > B"=OneMinus(None=Saturate(None=@n9)), "A == B"=@n10, "A < B"=@n10)
  FunctionOutput(None=Multiply(
    A=FunctionOutput(None=@n6) {outputName=Max Slope Mask, sortPriority=2}
    B=FunctionOutput(None=@n11) {outputName=Min Slope Mask, sortPriority=1}
   )) {outputName=Combined Slope Masks}
  NamedRerouteDeclaration(None=FunctionInput() {bUsePreviewValueAsDefault=true, inputName=Min Slope, inputType=FunctionInput_Scalar}) {name=Min Slope, nodeColor=<rgba 0 0.5078 1 1>}
  n12 = FunctionInput() {bUsePreviewValueAsDefault=true, inputName=Slope Falloff, inputType=FunctionInput_Scalar, sortPriority=2, previewValue=<1 0 0 1>}
  NamedRerouteDeclaration(None=@n12) {name=Slope Falloff, nodeColor=<rgba 0.8281 1 0 1>}
  n13 = FunctionInput() {bUsePreviewValueAsDefault=true, inputName=Max Slope, inputType=FunctionInput_Scalar, sortPriority=1, previewValue=<1 0 0 1>}
  NamedRerouteDeclaration(None=@n13) {name=Max Slope, nodeColor=<rgba 0.8516 0 1 1>}
 ]}
