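# Unlit Landmass Blueprint brush material that paints procedural noise into a landscape weightmap.
# Brush edge falloff and optional elevation and slope masks provide masking, alongside seeded noise with tiling, inversion, floor, contrast and multiply controls.
# Noise either modulates the mask edge through HeightLerp or multiplies it, with MultiplyNoise selecting between them.
# BrushBlendModes offers alpha, min, max and additive blends, with the previous weightmap point-sampled at texel centers.
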Material /Landmass/Landscape/BlueprintBrushes/Materials/BrushMaterials/M_Brush_ProceduralNoise_Weightmap
graph {
 properties {pixelDepthOffsetMode=PDOM_Legacy, shadingModel=MSM_Unlit, usedShadingModels=Unlit}
 nodes [
  Constant() {r=1}
  n1 = NamedRerouteUsage()
  n2 = NamedRerouteUsage()
  MaterialFunctionCall!(
   AlphaBlend=LinearInterpolate(A=NamedRerouteUsage(), B=NamedRerouteUsage(), Alpha=NamedRerouteUsage())
   Min=Min(A=NamedRerouteUsage(), B=LinearInterpolate(B=@n1, Alpha=@n2) {constA=1})
   Max=Max(A=NamedRerouteUsage(), B=NamedRerouteUsage())
   Additive=Add(A=NamedRerouteUsage(), B=NamedRerouteUsage())
  ) {materialFunction=/Landmass/Landscape/BlueprintBrushes/MF/BrushBlendModes}
  n3 = ScalarParameter() {defaultValue=10, parameterName=MinElevation}
  n4 = ScalarParameter() {defaultValue=75, parameterName=MaxElevation}
  n5 = ScalarParameter() {defaultValue=1, parameterName=ElevationFalloff}
  NamedRerouteDeclaration(None=LinearInterpolate(
    B=MaterialFunctionCall!(Min Elevation=@n3, Max Elevation=@n4, Falloff Width=@n5, Noise) {materialFunction=/Landmass/Landscape/BlueprintBrushes/Materials/BrushMaterials/BasicFunctions/LandmassBrush_ElevationMask}.Combined Masks
    Alpha=ScalarParameter() {defaultValue=1, parameterName=EnableElevationMask}
   ) {constA=1}) {name=Elevation Mask, nodeColor=<rgba 0 0.4141 1 1>}
  n6 = MaterialFunctionCall!(Falloff Width Override, Falloff Offset Override) {materialFunction=/Landmass/Landscape/BlueprintBrushes/Materials/BrushMaterials/BasicFunctions/LandmassBrush_ShapeMask}
  NamedRerouteDeclaration(None=@n6.Brush Base Height) {name=Current Height, nodeColor=<rgba 0.1484 1 0 1>}
  n7 = NamedRerouteUsage()
  n8 = NamedRerouteUsage()
  n9 = ScalarParameter() {parameterName=MultiplyNoise}
  NamedRerouteDeclaration(None=Saturate(None=LinearInterpolate(A=@n7, B=@n8, Alpha=@n9))) {name=Noise Mask, nodeColor=<rgba 1 0 0.0234 1>}
  n10 = NamedRerouteUsage()
  n11 = NamedRerouteUsage()
  NamedRerouteDeclaration(None=Min(A=Min(A=@n10, B=@n11), B=NamedRerouteUsage())) {name=Combined Mask, nodeColor=<rgba 0 0.3437 1 1>}
  NamedRerouteDeclaration(None=@n6.Brush Edge Falloff) {name=Brush Edge Mask, nodeColor=<rgba 1 0 0.1641 1>}
  n12 = ScalarParameter() {defaultValue=10, parameterName=MinSlope}
  n13 = ScalarParameter() {defaultValue=75, parameterName=MaxSlope}
  n14 = ScalarParameter() {defaultValue=1, parameterName=SlopeFalloff}
  NamedRerouteDeclaration(None=LinearInterpolate(
    B=MaterialFunctionCall!(Min Slope=@n12, Max Slope=@n13, Slope Falloff=@n14) {materialFunction=/Landmass/Landscape/BlueprintBrushes/Materials/BrushMaterials/BasicFunctions/LandmassBrush_SlopeMask}.Combined Slope Masks
    Alpha=ScalarParameter() {defaultValue=1, parameterName=EnableSlopeMask}
   ) {constA=1}) {name=Slope Mask, nodeColor=<rgba 0.7344 1 0 1>}
  n15 = MaterialFunctionCall() {materialFunction=/Landmass/Landscape/BlueprintBrushes/MF/LandmassBrush_UVs}
  n16 = ScalarParameter() {defaultValue=2, parameterName=WeightNoiseTiling}
  n17 = MaterialFunctionCall!(
   Position=AppendVector(A=Multiply(A=@n15.UVs, B=@n16), B=ScalarParameter() {defaultValue=1, parameterName=WeightNoiseSeed})
   Octaves=ScalarParameter() {defaultValue=4, parameterName=WeightNoiseOctaves}
   Persistence=ScalarParameter() {defaultValue=0.35, parameterName=WeightNoisePersistence}
   Turbulence=ScalarParameter() {parameterName=WeightNoiseTurbulence}
  ) {materialFunction=/Landmass/Landscape/BlueprintBrushes/MF/LandmassBrush_Noise}
  n18 = LinearInterpolate(A=@n17.Result, B=OneMinus(None=@n17.Result), Alpha=ScalarParameter() {parameterName=WeightInvertNoise})
  n19 = ScalarParameter() {parameterName=WeightNoiseFloor}
  n20 = Max(A=Divide(A=Subtract(A=@n18, B=@n19), B=OneMinus(None=@n19))) {constB=0}
  n21 = Power(Base=@n20, Exp=ScalarParameter() {defaultValue=1, parameterName=WeightNoiseContrast})
  n22 = ScalarParameter() {defaultValue=1, parameterName=WeightNoiseMultiply}
  NamedRerouteDeclaration(None=Saturate(None=Multiply(A=@n21, B=@n22))) {name=Noise Value, nodeColor=<rgba 0.094 0.3275 1 1>}
  NamedRerouteDeclaration(None=@n6.Previous Height) {name=Previous Height, nodeColor=<rgba 0 0.4609 1 1>}
  NamedRerouteDeclaration(None=TextureSampleParameter2D(UVs=MaterialFunctionCall() {materialFunction=/Landmass/Landscape/BlueprintBrushes/MF/PointSampleLandscapeUVs}.TexelCenterUVs) {parameterName=WeightmapRT, texture=/Engine/EngineResources/Black_Low.Black}.R) {name=Previous Weightmap, nodeColor=<rgba 0 0.4609 1 1>}
  n23 = Clamp(None=NamedRerouteUsage()) {maxDefault=0.999, minDefault=0.001}
  n24 = ScalarParameter() {defaultValue=1, parameterName=WeightNoiseContrast}
  NamedRerouteDeclaration(None=MaterialFunctionCall(Transition Phase=NamedRerouteUsage(), Height Texture=LinearInterpolate(B=@n23, Alpha=@n24) {constA=0.5}, Contrast=@n24) {materialFunction=&HeightLerp}.Alpha) {name=Alpha Modulated Edge, nodeColor=<rgba 0 1 0.7891 1>}
  NamedRerouteDeclaration(None=Multiply(A=NamedRerouteUsage(), B=NamedRerouteUsage())) {name=Alpha Multiplied by Noise, nodeColor=<rgba 0 1 0.0391 1>}
  n25 = ScalarParameter() {defaultValue=1, parameterName=WeightNoiseMultiply}
  NamedRerouteDeclaration(None=Saturate(None=Multiply(A=@n20, B=@n25))) {name=Noise Value Heightlerp, desc="Skip Contrast Power\r\nIt is applied differently for heightlerp", nodeColor=<rgba 0.094 0.3275 1 1>}
  NamedRerouteUsage()
 ]
}
