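# Unlit Landmass landscape brush material that adds a displacement texture on top of the existing landscape height.
# Samples the Displacement texture parameter with brush UVs and multiplies its red channel by the Height scalar.
# Adds that to the previous height from LandmassBrush_ShapeMask, then packs the sum with PackRG8Height.
# Only the previous-height output of the shape mask is used, so no falloff mask is applied to the addition.
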Material /Landmass/Landscape/BlueprintBrushes/Materials/BrushMaterials/M_Brush_SimpleAdd
graph {
 properties {floatPrecisionMode=MFPM_Full, pixelDepthOffsetMode=PDOM_Legacy, shadingModel=MSM_Unlit, usedShadingModels=Unlit}
 nodes [
  n1 = MaterialFunctionCall!(Falloff Width Override, Falloff Offset Override) {materialFunction=/Landmass/Landscape/BlueprintBrushes/Materials/BrushMaterials/BasicFunctions/LandmassBrush_ShapeMask}
  n2 = TextureSampleParameter2D(UVs=MaterialFunctionCall() {materialFunction=/Landmass/Landscape/BlueprintBrushes/MF/LandmassBrush_UVs}.UVs) {
   parameterName=Displacement
   samplerType=SAMPLERTYPE_LinearColor
   texture=/Engine/Functions/Engine_MaterialFunctions02/ExampleContent/Textures/SphereRenderHeightMap
  }
  n3 = ScalarParameter() {defaultValue=1000, parameterName=Height}
  MaterialFunctionCall!(Height Int -32k to 32k=Add(A=NamedRerouteDeclaration(None=@n1.Previous Height) {name=Previous Height, nodeColor=<rgba 0 0.4609 1 1>}, B=Multiply(A=@n2.R, B=@n3))) {materialFunction=/Landmass/Landscape/BlueprintBrushes/MF/PackRG8Height}
 ]
}
