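# FractalTerraces is a Landmass landscape brush material function that reshapes a scalar Height into irregular, smoothed terraces.
# A single Custom HLSL node hashes the height band index with Rand3DPCG16, randomly halving Interval up to three nested times to vary step sizes.
# Deeper subdivisions also adjust Smoothness by a power.
# A mirrored power S-curve within each step then shapes the transition between levels, controlled by Smoothness.
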MaterialFunction /Landmass/Landscape/BlueprintBrushes/MF/FractalTerraces
graph {nodes [
  MaterialExpressionFunctionInput_0 = FunctionInput() {inputName=Height, inputType=FunctionInput_Scalar}
  MaterialExpressionFunctionInput_2 = FunctionInput() {bUsePreviewValueAsDefault=true, inputName=Interval, inputType=FunctionInput_Scalar, sortPriority=1, previewValue=<0.5 0 0 1>}
  MaterialExpressionFunctionInput_1 = FunctionInput() {bUsePreviewValueAsDefault=true, inputName=Smoothness, inputType=FunctionInput_Scalar, sortPriority=2}
  n1 = Custom!(InHeight=@MaterialExpressionFunctionInput_0, Interval=@MaterialExpressionFunctionInput_2, Smoothness=@MaterialExpressionFunctionInput_1) {
   code="\r\n\r\n\r\n\r\nfloat startrandlevel = floor(InHeight / Interval);\r\n\r\nuint randsublevels = Rand3DPCG16(int3(startrandlevel,0,0)).x;\r\n\r\nfloat testrand = randsublevels / 65536.0;\r\n\r\nif( testrand > 0.5)\r\n{\r\n\tInterval /= 2;\r\n\t\r\n\tstartrandlevel = floor(InHeight / Interval);\r\n\r\n\trandsublevels = Rand3DPCG16(int3(startrandlevel,0,0)).x;\r\n\r\n\ttestrand = randsublevels / 65536.0;\r\n\t\r\n\t\r\n\t\r\n\tif( testrand > 0.5)\r\n\t{\r\n\t\tInterval /= 2;\r\n\t\tSmoothness = pow(Smoothness, 0.75);\r\n\t\t\r\n\t\tstartrandlevel = floor(InHeight / Interval);\r\n\r\n\t\trandsublevels = Rand3DPCG16(int3(startrandlevel+69.0,0,0)).x;\r\n\r\n\t\ttestrand = randsublevels / 65536.0;\r\n\t\t\r\n\t\t\tif( testrand > 0.5)\r\n\t\t\t{\r\n\t\t\t\tInterval /= 2;\r\n\t\t\t\tSmoothness = pow(Smoothness, 0.75);\r\n\t\t\t\t\r\n\t\t\t\tstartrandlevel = floor(InHeight / Interval);\r\n\t\t\r\n\t\t\t\trandsublevels = Rand3DPCG16(int3(startrandlevel+69.0,0,0)).x;\r\n\t\t\r\n\t\t\t\ttestrand = randsublevels / 65536.0;\r\n\t\t\t\t\r\n\t\t\t\t\r\n\t\t\t}\r\n\t\t\r\n\t\t\r\n\t}\r\n}\r\n\r\nSmoothness = 1 / max(Smoothness, 0.0001);\r\n\r\nfloat curlevel = InHeight / Interval;\r\nfloat curphase = frac(curlevel);\r\nfloat halfmask = saturate(ceil(curphase - 0.5));\r\ncurlevel = floor(curlevel) * Interval;\r\n\r\n\r\nfloat scurve = lerp(curphase, 1-curphase, halfmask) * 2;\r\nscurve = pow(scurve, Smoothness) / 2;\r\nscurve = lerp(scurve, 1-scurve, halfmask) * Interval;\r\nreturn scurve + curlevel;"
   description=Smooth Terracing
   outputType=CMOT_Float1
   inputs [
    {inputName=InHeight, input {expression=&self:MaterialExpressionFunctionInput_0}}
    {inputName=Interval, input {expression=&self:MaterialExpressionFunctionInput_2}}
    {inputName=Smoothness, input {expression=&self:MaterialExpressionFunctionInput_1}}
   ]
  }
  FunctionOutput(None=@n1)
 ]}
