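# Unlit Landmass landscape brush material that carves terraces into landscape height.
# It unpacks the height render target, runs offset height through FractalTerraces with a noise-varied interval, then blends by mask and opacity and repacks to RG8.
# Named reroutes define brush height, an S-curved edge-falloff or shape-mask alpha, edge noise, displacement and noise-broken weightmap masks.
# Some scratch noise nodes with placeholder parameter names feed nothing.
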Material /Landmass/Landscape/BlueprintBrushes/Materials/BrushMaterials/M_Brush_Terraces
graph {
 properties {pixelDepthOffsetMode=PDOM_Legacy, shadingModel=MSM_Unlit, usedShadingModels=Unlit}
 nodes [
  Constant()
  LinearInterpolate(B=NamedRerouteUsage(), Alpha=ScalarParameter() {parameterName=DisplacementAsAlpha}) {constA=1}
  n1 = NamedRerouteUsage()
  n2 = NamedRerouteUsage()
  n3 = Multiply(
   A=MaterialFunctionCall() {materialFunction=/Landmass/Landscape/BlueprintBrushes/MF/PointSampleLandscapeUVs}.TexelCenterUVs
   B=ScalarParameter() {defaultValue=32, parameterName=hfd}
  )
  n4 = VectorParameter() {parameterName=vasdfasfsdf, defaultValue=<rgba 10 0.5 0 0>}
  n5 = Multiply(
   A=MaterialFunctionCall!(Position=AppendVector(A=@n3, B=@n4.B), Octaves=@n4.R, Persistence=@n4.R, Turbulence) {materialFunction=/Landmass/Landscape/BlueprintBrushes/MF/LandmassBrush_Noise}.Result
   B=ScalarParameter() {defaultValue=4096, parameterName=asd}
  )
  n6 = MaterialFunctionCall!(
   Height=Subtract(A=@n1, B=@n2)
   Interval=Add(A=ScalarParameter() {defaultValue=4096, parameterName=Terrace Frequency}, B=@n5)
   Smoothness=ScalarParameter() {defaultValue=0.5, parameterName=Terrace Smoothness}
  ) {materialFunction=/Landmass/Landscape/BlueprintBrushes/MF/FractalTerraces}
  n7 = NamedRerouteUsage()
  n8 = ScalarParameter() {defaultValue=1, parameterName=Opacity}
  MaterialFunctionCall!(Height Int -32k to 32k=LinearInterpolate(A=NamedRerouteUsage(), B=Add(A=@n6.Result, B=@n2), Alpha=Multiply(A=@n7, B=@n8))) {materialFunction=/Landmass/Landscape/BlueprintBrushes/MF/PackRG8Height}
  n9 = Multiply(A=Multiply(A=@n1, B=Constant3Vector() {constant=<rgba 1 1.618 0.707 0>}), B=ScalarParameter() {defaultValue=0, parameterName=asdf})
  n10 = VectorParameter() {parameterName=adsfasfd, defaultValue=<rgba 3 0 0 0>}
  Multiply(
   A=MaterialFunctionCall(Position=Add(A=@n9, B=@n5), Octaves=@n10.RGB, Persistence=@n10.RGB) {materialFunction=&LandmassBrush_Noise}.Result
   B=ScalarParameter() {defaultValue=4096, parameterName=ghghdfh}
  )
  n11 = TextureSampleParameter2D(UVs=MaterialFunctionCall() {materialFunction=/Landmass/Landscape/BlueprintBrushes/MF/LandmassBrush_UVs}.UVs) {parameterName=ShapeMask, texture=/Engine/ArtTools/RenderToTexture/Textures/T_EV_BlankWhite_01}
  n12 = NamedRerouteUsage()
  n13 = Divide(A=Subtract(A=Subtract(A=@n11.R) {constB=0.5}, B=Divide(A=@n12)), B=ScalarParameter() {defaultValue=0.1, parameterName=FalloffWidth})
  NamedRerouteDeclaration(None=Saturate(None=Divide(A=@n13) {constB=0.5})) {name=ShapeMask, nodeColor=<rgba 1 0 0.5859 1>}
  NamedRerouteDeclaration(None=Add(
    A=MaterialFunctionCall!(Enable) {materialFunction=/Landmass/Landscape/BlueprintBrushes/MF/Landmass_BrushTilt}.Tilt
    B=ScalarParameter() {parameterName=BrushZ}
   )) {name=Current Height, nodeColor=<rgba 0.1484 1 0 1>}
  n14 = NamedRerouteUsage()
  n15 = NamedRerouteUsage()
  n16 = ScalarParameter() {parameterName=UseShapeMask}
  NamedRerouteDeclaration(None=MaterialFunctionCall(In=LinearInterpolate(A=@n14, B=@n15, Alpha=@n16), Power=ScalarParameter() {defaultValue=1, parameterName=FalloffPower}) {materialFunction=&SCurve}.Result) {name=Brush Alpha, nodeColor=<rgba 1 0 0.1641 1>}
  n17 = ScalarParameter() {defaultValue=0.1, parameterName=FalloffWidth}
  n18 = ScalarParameter() {defaultValue=0.2, parameterName=FalloffRadius}
  n19 = Add(A=NamedRerouteUsage(), B=ScalarParameter() {parameterName=FalloffOffset})
  NamedRerouteDeclaration(None=Saturate(None=MaterialFunctionCall!(UVs, Falloff Width=@n17, Falloff Radius=@n18, Offset=@n19) {materialFunction=/Landmass/Landscape/BlueprintBrushes/MF/LandmassBrush_EdgeFalloff}.Edge Mask)) {name=Default Edge Falloff, nodeColor=<rgba 0.8281 0 1 1>}
  n20 = VectorParameter() {parameterName=DisplacementChannel, defaultValue=<rgba 1 0 0 0>}
  n21 = AppendVector(A=@n20.RGB, B=@n20.RGB)
  n22 = TextureSampleParameter2D(UVs=MaterialFunctionCall() {materialFunction=&LandmassBrush_UVs}.UVs) {parameterName=Displacement, texture=&T_EV_BlankWhite_01}
  n23 = ScalarParameter() {parameterName=DisplacementReferencePlane}
  NamedRerouteDeclaration(None=Add(A=DotProduct(A=@n21, B=@n22.RGBA), B=OneMinus(None=@n23))) {name=Displacement Mask, nodeColor=<rgba 0 0.8125 1 1>}
  n24 = VectorParameter() {parameterName=DisplacementChannel, defaultValue=<rgba 1 0 0 0>}
  n25 = MaterialFunctionCall() {materialFunction=&LandmassBrush_UVs}
  NamedRerouteDeclaration(None=DotProduct(
    A=AppendVector(A=@n24.RGB, B=@n24.RGB)
    B=TextureSampleParameter2D(UVs=@n25.UVs) {parameterName=WeightmapTexture, texture=&T_EV_BlankWhite_01}.RGBA
   )) {name=WeightmapMask, nodeColor=<rgba 1 0 0.5859 1>}
  n26 = AppendVector(
   A=Multiply(
    A=MaterialFunctionCall() {materialFunction=&LandmassBrush_UVs}.UVs
    B=ScalarParameter() {defaultValue=2, parameterName=EdgeNoiseTiling}
   )
   B=ScalarParameter() {defaultValue=1, parameterName=EdgeNoiseSeed}
  )
  n27 = ScalarParameter() {defaultValue=4, parameterName=EdgeNoiseOctaves}
  n28 = ScalarParameter() {defaultValue=0.35, parameterName=EdgeNoisePersistence}
  NamedRerouteDeclaration(None=Multiply(
    A=MaterialFunctionCall(Position=@n26, Octaves=@n27, Persistence=@n28) {materialFunction=&LandmassBrush_Noise}.Result
    B=ScalarParameter() {defaultValue=0.2, parameterName=EdgeNoiseStrength}
   )) {name=Edge Falloff Noise, nodeColor=<rgba 0.0786 0.3441 1 1>}
  n29 = Saturate(None=MaterialFunctionCall!(Weightmap Mask=NamedRerouteUsage(), Mask=NamedRerouteUsage()) {materialFunction=/Landmass/Landscape/BlueprintBrushes/MF/BrushBlendModes_Weight}.Result)
  n30 = NamedRerouteUsage()
  n31 = ScalarParameter() {defaultValue=1, parameterName=WeightNoiseContrast}
  n32 = MaterialFunctionCall(Transition Phase=@n29, Height Texture=@n30, Contrast=@n31) {materialFunction=&HeightLerp}
  n33 = Multiply(A=@n29, B=Power(Base=@n30, Exp=@n31))
  n34 = ScalarParameter() {parameterName=MultiplyNoise}
  NamedRerouteDeclaration(None=Multiply(A=LinearInterpolate(A=@n32.Alpha, B=@n33, Alpha=@n34), B=ScalarParameter() {defaultValue=1, parameterName=WeightmapOpacity})) {name=Weightmap Result, nodeColor=<rgba 0.5234 0 1 1>}
  n35 = MaterialFunctionCall() {materialFunction=&LandmassBrush_UVs}
  n36 = ScalarParameter() {defaultValue=8, parameterName=WeightNoiseTiling}
  n37 = MaterialFunctionCall(
   Position=AppendVector(A=Multiply(A=@n35.UVs, B=@n36), B=ScalarParameter() {defaultValue=1, parameterName=WeightNoiseSeed})
   Octaves=ScalarParameter() {defaultValue=4, parameterName=WeightNoiseOctaves}
   Persistence=ScalarParameter() {defaultValue=0.35, parameterName=WeightNoisePersistence}
   Turbulence=ScalarParameter() {parameterName=WeightNoiseTurbulence}
  ) {materialFunction=&LandmassBrush_Noise}
  n38 = OneMinus(None=@n37.Result)
  n39 = ScalarParameter() {parameterName=WeightNoiseInvert}
  NamedRerouteDeclaration(None=Clamp(None=LinearInterpolate(A=@n37.Result, B=@n38, Alpha=@n39)) {maxDefault=0.999, minDefault=0.001}) {name=Weightmap Noise, nodeColor=<rgba 0.1719 0 1 1>}
  n40 = TextureSampleParameter2D(UVs=MaterialFunctionCall() {materialFunction=&PointSampleLandscapeUVs}.TexelCenterUVs) {parameterName=HeightRT, texture=/Engine/EngineResources/Black_Low.Black}
  NamedRerouteDeclaration(None=MaterialFunctionCall!(RG=ComponentMask(None=@n40.RGB) {g=true, r=true}) {materialFunction=/Landmass/Landscape/OldPrototype_BP/Mats/MF/UnpackRG8Height}.Result) {name=Previous Height, nodeColor=<rgba 0 0.4609 1 1>}
  NamedRerouteUsage()
  NamedRerouteUsage()
 ]
}
