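# Unlit Landmass landscape brush material that stamps a texture into either terrain height or a paint-layer weightmap, chosen by the Weightmap static switch.
# LandmassBrush_ShapeMask supplies base height, previous height and edge falloff.
# Displacement and weightmap textures share brush UVs and a DisplacementChannel dot-product selector.
# Height blend results are packed to RG8, while weight blending draws on masks, with depth-based falloff and noise also computed.
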Material /Landmass/Landscape/BlueprintBrushes/Materials/BrushMaterials/M_Brush_TextureStamp
graph {
 properties {floatPrecisionMode=MFPM_Full, pixelDepthOffsetMode=PDOM_Legacy, shadingModel=MSM_Unlit, usedShadingModels=Unlit}
 nodes [
  n1 = MaterialFunctionCall!(Falloff Width Override, Falloff Offset Override) {materialFunction=/Landmass/Landscape/BlueprintBrushes/Materials/BrushMaterials/BasicFunctions/LandmassBrush_ShapeMask}
  NamedRerouteDeclaration(None=@n1.Brush Base Height) {name=Current Height, nodeColor=<rgba 0.1484 1 0 1>}
  n2 = LinearInterpolate(B=NamedRerouteUsage(), Alpha=ScalarParameter() {defaultValue=1, parameterName=RelativeDepth}) {constA=1}
  n3 = ScalarParameter() {parameterName=FalloffOffset}
  n4 = ScalarParameter() {defaultValue=0.1, parameterName=FalloffWidth}
  n5 = Saturate(None=Divide(A=Subtract(A=@n2, B=@n3), B=Abs(None=@n4)))
  n6 = NamedRerouteUsage()
  NamedRerouteDeclaration(None=LinearInterpolate(A=@n5, B=Multiply(A=@n5, B=@n6), Alpha=ScalarParameter() {parameterName=EnableDefaultFalloff}) {constA=1}) {name=Weightmap Alpha, nodeColor=<rgba 0.078 0.5071 1 1>}
  n7 = MaterialFunctionCall!(Brush Alpha=NamedRerouteUsage()) {
   desc="Struct: Landmass_NoisePass_WeightMapSettings\r\nMacro: "
   materialFunction=/Landmass/Landscape/BlueprintBrushes/Materials/BrushMaterials/BasicFunctions/M_Brush_WeightmapNoise
  }
  NamedRerouteDeclaration(None=Multiply(A=@n7.Weightmap Noise, B=ScalarParameter() {defaultValue=1, parameterName=WeightmapOpacity})) {name=Weightmap Alpha with Noise, nodeColor=<rgba 0.5 1 0 1>}
  NamedRerouteDeclaration(None=@n1.Brush Edge Falloff) {name=Brush Edge Mask, nodeColor=<rgba 1 0 0.1641 1>}
  n8 = VectorParameter() {parameterName=DisplacementChannel, defaultValue=<rgba 1 0 0 0>}
  n9 = AppendVector(A=@n8.RGB, B=@n8.RGB)
  n10 = TextureSampleParameter2D(UVs=MaterialFunctionCall() {materialFunction=/Landmass/Landscape/BlueprintBrushes/MF/LandmassBrush_UVs}.UVs) {parameterName=Displacement, texture=/Engine/ArtTools/RenderToTexture/Textures/T_EV_BlankWhite_01}
  n11 = ScalarParameter() {parameterName=DisplacementReferencePlane}
  NamedRerouteDeclaration(None=Add(A=DotProduct(A=@n9, B=@n10.RGBA), B=OneMinus(None=@n11))) {name=Displacement Mask, nodeColor=<rgba 0 0.8125 1 1>}
  n12 = VectorParameter() {parameterName=DisplacementChannel, defaultValue=<rgba 1 0 0 0>}
  n13 = MaterialFunctionCall() {materialFunction=&LandmassBrush_UVs}
  NamedRerouteDeclaration(None=DotProduct(
    A=AppendVector(A=@n12.RGB, B=@n12.RGB)
    B=TextureSampleParameter2D(UVs=@n13.UVs) {parameterName=WeightmapTexture, texture=&T_EV_BlankWhite_01}.RGBA
   )) {name=WeightmapMask, nodeColor=<rgba 1 0 0.5859 1>}
  n14 = NamedRerouteUsage()
  n15 = NamedRerouteUsage()
  NamedRerouteDeclaration(None=Saturate(None=MaterialFunctionCall!(Weightmap Mask=@n14, Mask=@n15) {materialFunction=/Landmass/Landscape/BlueprintBrushes/MF/BrushBlendModes_Weight}.Result)) {name=Weightmap Result, nodeColor=<rgba 0.5234 0 1 1>}
  NamedRerouteDeclaration(None=@n1.Previous Height) {name=Previous Height, nodeColor=<rgba 0 0.4609 1 1>}
  n16 = MaterialFunctionCall!(Current Height=NamedRerouteUsage(), Edge Mask=NamedRerouteUsage(), Displacement Mask=NamedRerouteUsage()) {materialFunction=/Landmass/Landscape/BlueprintBrushes/MF/BrushBlendModes_Height}
  n17 = ScalarParameter() {parameterName=DebugOffset}
  StaticSwitchParameter(
   True=NamedRerouteUsage()
   False=MaterialFunctionCall!(Height Int -32k to 32k=Add(A=@n16.Result, B=@n17)) {materialFunction=/Landmass/Landscape/BlueprintBrushes/MF/PackRG8Height}.XY
  ) {parameterName=Weightmap}
 ]
}
