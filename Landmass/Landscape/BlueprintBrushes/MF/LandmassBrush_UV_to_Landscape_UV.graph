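# Landmass material function that converts blueprint brush UVs into landscape render target UVs.
# It centers the input UV, scales it by brush size and scale, rotates it using BrushXVector and its derived perpendicular, then offsets it by BrushPosition.
# The world XY result is made relative to LandscapeLocation, shifted by half of LandscapeScale and divided by RTWorldSize, all read from MPC_Landscape.
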MaterialFunction /Landmass/Landscape/BlueprintBrushes/MF/LandmassBrush_UV_to_Landscape_UV
graph {nodes [
  n1 = FunctionInput(Preview=MaterialFunctionCall() {materialFunction=&/LandmassBrush_UVs}.UVs) {bUsePreviewValueAsDefault=true, inputName=BrushUV, inputType=FunctionInput_Vector2}
  n2 = Multiply(
   A=VectorParameter() {parameterName=BrushScale, defaultValue=<rgba 1 1 1 0>}.RGB
   B=ScalarParameter() {defaultValue=4096, parameterName=BrushSize}
  )
  n3 = MaterialFunctionCall(Float2=Multiply(A=Subtract(A=@n1) {constB=0.5}, B=ComponentMask(None=@n2) {g=true, r=true})) {materialFunction=&BreakOutFloat2Components}
  n4 = VectorParameter() {parameterName=BrushXVector, defaultValue=<rgba 1 0 0 0>}
  n5 = Multiply(A=@n3.R, B=ComponentMask(None=@n4.RGB) {g=true, r=true})
  n6 = Multiply(
   A=@n3.G
   B=MaterialFunctionCall(X=Multiply(A=@n4.G) {constB=-1}, Y=@n4.R) {materialFunction=&MakeFloat2}.Result
  )
  n7 = VectorParameter() {parameterName=BrushPosition}
  n8 = Subtract(
   A=Add(A=Add(A=@n5, B=@n6), B=ComponentMask(None=@n7.RGB) {g=true, r=true})
   B=ComponentMask(None=CollectionParameter() {collection=/Landmass/Landscape/BlueprintBrushes/MPC/MPC_Landscape, parameterName=LandscapeLocation}) {g=true, r=true}
  )
  n9 = Multiply(A=ComponentMask(None=CollectionParameter() {collection=&MPC_Landscape, parameterName=LandscapeScale}) {g=true, r=true}) {constB=0.5}
  n10 = CollectionParameter() {collection=&MPC_Landscape, parameterName=RTWorldSize}
  FunctionOutput(None=Divide(A=Add(A=@n8, B=@n9), B=ComponentMask(None=@n10) {g=true, r=true}))
 ]}
