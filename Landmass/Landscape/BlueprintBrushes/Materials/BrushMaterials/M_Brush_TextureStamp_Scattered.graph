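# Unlit Landmass brush material that stamps a displacement texture in scattered patches into the landscape heightmap.
# Brush UVs are tiled into cells, with seeded vector noise supplying per-cell randomness and named masks for edge falloff, patch noise and weightmaps.
# Stamp UVs are scaled, offset and rotated, then the displacement is smoothed and edge-masked.
# Point-sampled height plus an offset blends with previous height through brush blend modes and is repacked to RG8.
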Material /Landmass/Landscape/BlueprintBrushes/Materials/BrushMaterials/M_Brush_TextureStamp_Scattered
graph {
 properties {floatPrecisionMode=MFPM_Full, pixelDepthOffsetMode=PDOM_Legacy, shadingModel=MSM_Unlit, usedShadingModels=Unlit}
 nodes [
  n1 = TextureSampleParameter2D(UVs=MaterialFunctionCall() {materialFunction=/Landmass/Landscape/BlueprintBrushes/MF/PointSampleLandscapeUVs}.TexelCenterUVs) {parameterName=HeightRT, texture=/Engine/EngineResources/Black_Low.Black}
  n2 = TextureSampleParameter2D(UVs=MaterialFunctionCall!(UVs=NamedRerouteUsage(), Texture Size=NamedRerouteUsage()) {materialFunction=/Engine/Functions/Engine_MaterialFunctions01/Texturing/PointSampledUVs}.Result) {parameterName=HeightRT, texture=&Black}
  n3 = NamedRerouteUsage()
  n4 = NamedRerouteUsage()
  n5 = LinearInterpolate(
   A=NamedRerouteDeclaration(None=MaterialFunctionCall!(RG=@n1.RGB) {materialFunction=/Landmass/Landscape/BlueprintBrushes/MF/UnpackRG8Height}.Result) {name=PreviousHeight, nodeColor=<rgba 0.7578 1 0 1>}
   B=Add(A=MaterialFunctionCall(RG=@n2.RGB) {materialFunction=&UnpackRG8Height}.Result, B=NamedRerouteUsage())
   Alpha=Multiply(A=Multiply(A=@n3, B=@n4), B=ScalarParameter() {defaultValue=1, parameterName=Opacity})
  ) {constAlpha=0}
  n6 = Min(A=@n5, B=NamedRerouteUsage())
  n7 = Max(A=@n5, B=NamedRerouteUsage())
  n8 = Add(A=NamedRerouteUsage(), B=NamedRerouteUsage())
  AppendVector(
   A=MaterialFunctionCall!(Height Int -32k to 32k=MaterialFunctionCall!(AlphaBlend=@n5, Min=@n6, Max=@n7, Additive=@n8) {materialFunction=/Landmass/Landscape/BlueprintBrushes/MF/BrushBlendModes}.Result) {materialFunction=/Landmass/Landscape/BlueprintBrushes/MF/PackRG8Height}.XY
   B=TextureSampleParameter2D() {parameterName=HeightRT, texture=&Black}.B
  )
  n9 = MaterialFunctionCall!(Falloff Width Override, Falloff Offset Override) {materialFunction=/Landmass/Landscape/BlueprintBrushes/Materials/BrushMaterials/BasicFunctions/LandmassBrush_ShapeMask}
  NamedRerouteDeclaration(None=@n9.Brush Base Height) {name=Current Height, nodeColor=<rgba 0.1484 1 0 1>}
  n10 = LinearInterpolate(B=NamedRerouteUsage(), Alpha=ScalarParameter() {defaultValue=1, parameterName=RelativeDepth}) {constA=1}
  n11 = ScalarParameter() {parameterName=FalloffOffset}
  n12 = ScalarParameter() {defaultValue=0.01, parameterName=FalloffWidth}
  n13 = Saturate(None=Divide(A=Subtract(A=@n10, B=@n11), B=Abs(None=@n12)))
  n14 = NamedRerouteUsage()
  NamedRerouteDeclaration(None=LinearInterpolate(A=@n13, B=Multiply(A=@n13, B=@n14), Alpha=ScalarParameter() {parameterName=EnableDefaultFalloff}) {constA=1}) {name=Weightmap Alpha, nodeColor=<rgba 0.078 0.5071 1 1>}
  n15 = MaterialFunctionCall!(Brush Alpha=NamedRerouteUsage()) {
   desc="Struct: Landmass_NoisePass_WeightMapSettings\r\nMacro: "
   materialFunction=/Landmass/Landscape/BlueprintBrushes/Materials/BrushMaterials/BasicFunctions/M_Brush_WeightmapNoise
  }
  NamedRerouteDeclaration(None=Multiply(A=@n15.Weightmap Noise, B=ScalarParameter() {defaultValue=1, parameterName=WeightmapOpacity})) {name=Weightmap Alpha with Noise, nodeColor=<rgba 0.5 1 0 1>}
  NamedRerouteDeclaration(None=@n9.Brush Edge Falloff) {name=Brush Edge Mask, nodeColor=<rgba 1 0 0.1641 1>}
  n16 = VectorParameter() {parameterName=DisplacementChannel, defaultValue=<rgba 1 0 0 0>}
  n17 = AppendVector(A=@n16.RGB, B=@n16.RGB)
  n18 = TextureSampleParameter2D(UVs=MaterialFunctionCall() {materialFunction=/Landmass/Landscape/BlueprintBrushes/MF/LandmassBrush_UVs}.UVs) {parameterName=Displacement, texture=/Engine/Functions/Engine_MaterialFunctions02/PivotPainter2/Black_1x1_EXR_Texture}
  n19 = ScalarParameter() {defaultValue=0.5, parameterName=DisplacementReferencePlane}
  NamedRerouteDeclaration(None=Add(A=DotProduct(A=@n17, B=@n18.RGBA), B=OneMinus(None=@n19))) {name=Displacement Mask, nodeColor=<rgba 0 1 0.7422 1>}
  NamedRerouteDeclaration(None=AppendVector(
    A=CollectionParameter() {collection=/Landmass/Landscape/BlueprintBrushes/MPC/MPC_Landscape, parameterName=RTResX}
    B=CollectionParameter() {collection=&MPC_Landscape, parameterName=RTResY}
   )) {name=RTResolution, nodeColor=<rgba 0.7813 0 1 1>}
  NamedRerouteDeclaration(None=AppendVector(
    A=CollectionParameter() {collection=&MPC_Landscape, parameterName=LSQuadsX}
    B=CollectionParameter() {collection=&MPC_Landscape, parameterName=LSQuadsY}
   )) {name=LandscapeQuads, nodeColor=<rgba 0.7812 0 1 1>}
  n20 = Multiply(
   A=MaterialFunctionCall() {materialFunction=&LandmassBrush_UVs}.UVs
   B=ScalarParameter() {defaultValue=4, parameterName=Repeats}
  )
  n21 = Floor(None=@n20)
  n22 = Divide(
   A=Add(A=NamedRerouteDeclaration(None=@n21) {name=Tile Index, nodeColor=<rgba 1 0.9609 0 1>}) {constB=0.5}
   B=ScalarParameter() {defaultValue=4, parameterName=Repeats}
  ) {desc=UV position within Tile 0-1}
  NamedRerouteDeclaration(None=Saturate(None=MaterialFunctionCall!(BrushUV=@n22) {materialFunction=/Landmass/Landscape/BlueprintBrushes/MF/LandmassBrush_UV_to_Landscape_UV}.Result)) {name=UV In SourceHeight, nodeColor=<rgba 0 1 0.1797 1>}
  n23 = Frac(None=@n20)
  NamedRerouteDeclaration(None=@n23) {name=Sub UV Patches, nodeColor=<rgba 1 0.0938 0 1>}
  n24 = ComponentMask(None=NamedRerouteUsage()) {g=true, r=true}
  n25 = ScalarParameter() {defaultValue=0.2, parameterName=CellOffset}
  n26 = Add(
   A=NamedRerouteDeclaration(None=@n23) {name=Per Patch UVs, nodeColor=<rgba 0.4766 1 0 1>}
   B=NamedRerouteDeclaration(None=Multiply(A=@n24, B=@n25)) {name=Per Cell Center Offset, nodeColor=<rgba 0.2891 0 1 1>}
  )
  n27 = ScalarParameter() {defaultValue=0.1, parameterName=FalloffWidth}
  n28 = ScalarParameter() {defaultValue=0.3, parameterName=FalloffRadius}
  n29 = Add(A=NamedRerouteUsage(), B=ScalarParameter() {defaultValue=0.2, parameterName=CellOffset})
  NamedRerouteDeclaration(None=Saturate(None=MaterialFunctionCall!(UVs=@n26, Falloff Width=@n27, Falloff Radius=@n28, Offset=@n29) {materialFunction=/Landmass/Landscape/BlueprintBrushes/MF/LandmassBrush_EdgeFalloff}.Edge Mask)) {name=Per Cell Alpha, nodeColor=<rgba 0 0.8359 1 1>}
  n30 = MaterialFunctionCall() {materialFunction=&LandmassBrush_UVs}
  n31 = Multiply(A=ScalarParameter() {defaultValue=8, parameterName=EdgeNoiseTiling}, B=ScalarParameter() {defaultValue=4, parameterName=Repeats})
  n32 = MaterialFunctionCall!(
   Position=AppendVector(A=Multiply(A=@n30.UVs, B=@n31), B=ScalarParameter() {defaultValue=1, parameterName=EdgeNoiseSeed})
   Octaves=ScalarParameter() {defaultValue=8, parameterName=NoiseOctaves}
   Persistence=ScalarParameter() {defaultValue=0.35, parameterName=NoisePersistence}
   Turbulence
  ) {materialFunction=/Landmass/Landscape/BlueprintBrushes/MF/LandmassBrush_Noise}
  NamedRerouteDeclaration(None=Multiply(A=Subtract(A=@n32.Result) {constB=0.5}, B=ScalarParameter() {defaultValue=0.1, parameterName=EdgeNoiseStrength})) {name=PatchNoise, nodeColor=<rgba 1 0.2578 0 1>}
  n33 = VectorParameter() {parameterName=DisplacementChannel, defaultValue=<rgba 1 0 0 0>}
  n34 = MaterialFunctionCall() {materialFunction=&LandmassBrush_UVs}
  NamedRerouteDeclaration(None=DotProduct(
    A=AppendVector(A=@n33.RGB, B=@n33.RGB)
    B=TextureSampleParameter2D(UVs=@n34.UVs) {parameterName=WeightmapTexture, texture=/Engine/ArtTools/RenderToTexture/Textures/T_EV_BlankWhite_01}.RGBA
   )) {name=WeightmapMask, nodeColor=<rgba 1 0 0.5859 1>}
  n35 = AppendVector(A=Add(A=NamedRerouteUsage()) {constB=0.5}, B=ScalarParameter() {parameterName=CellSeed})
  NamedRerouteDeclaration(None=ConstantBiasScale(None=VectorNoise(World Position=@n35)) {bias=-0.5, scale=2}) {name=Per Cell Random V3, nodeColor=<rgba 0.1484 0 1 1>}
  n36 = Multiply(
   A=VectorParameter() {parameterName=BrushScale, defaultValue=<rgba 1 1 1 0>}.R
   B=ScalarParameter() {defaultValue=4096, parameterName=BrushSize}
  )
  n37 = ScalarParameter() {defaultValue=4, parameterName=Repeats}
  n38 = ScalarParameter() {defaultValue=1, parameterName=RelativeDepth}
  n39 = NamedRerouteUsage()
  NamedRerouteDeclaration(None=Multiply(A=Divide(A=@n36, B=@n37), B=Multiply(A=@n38, B=@n39))) {name=Displacement Height, nodeColor=<rgba 1 0.4922 0 1>}
  n40 = NamedRerouteUsage()
  n41 = MaterialFunctionCall(Float2=Abs(None=NamedRerouteUsage())) {materialFunction=&BreakOutFloat2Components}
  n42 = OneMinus(None=Max(A=@n41.R, B=@n41.R))
  n43 = Add(
   A=MaterialFunctionCall(UVs=NamedRerouteUsage(), Texture Scale=NamedRerouteDeclaration(None=@n42) {name=Scale Adjustment, nodeColor=<rgba 1 0.4922 0 1>}) {materialFunction=&ScaleUVsByCenter}.UVs
   B=NamedRerouteUsage()
  )
  n44 = Constant()
  n45 = ComponentMask(None=RotateAboutAxis(
    NormalizedRotationAxis=Constant3Vector() {constant=<rgba 0 0 1 0>}
    RotationAngle=ComponentMask(None=@n40) {b=true}
    PivotPoint=Constant() {r=0.5}
    Position=AppendVector(A=@n43, B=@n44)
   )) {g=true, r=true}
  n46 = TextureSampleParameter2D(UVs=Saturate(None=Add(A=@n45, B=@n43))) {
   parameterName=Displacement
   samplerType=SAMPLERTYPE_LinearColor
   texture=&Black_1x1_EXR_Texture
  }
  n47 = MaterialFunctionCall!(
   A=Constant()
   B=Subtract(A=@n46.R, B=ScalarParameter() {defaultValue=0.5, parameterName=DisplacementReferencePlane})
   K=ScalarParameter() {parameterName=Smoothness}
  ) {materialFunction=/Landmass/Landscape/BlueprintBrushes/MF/SmoothIntersections}
  n48 = NamedRerouteUsage()
  n49 = MaterialFunctionCall(UVs=NamedRerouteUsage(), Falloff Width=Constant() {r=0.02}, Falloff Radius=Constant(), Offset=Constant() {r=0.05}) {materialFunction=&LandmassBrush_EdgeFalloff}
  NamedRerouteDeclaration(None=Multiply(A=Multiply(A=@n47.Smooth Max, B=@n48), B=Saturate(None=@n49.Edge Mask))) {name=Patch Displacement, nodeColor=<rgba 1 0.6563 0 1>}
  n50 = NamedRerouteUsage()
  n51 = NamedRerouteUsage()
  NamedRerouteDeclaration(None=Saturate(None=MaterialFunctionCall!(Weightmap Mask=@n50, Mask=@n51) {materialFunction=/Landmass/Landscape/BlueprintBrushes/MF/BrushBlendModes_Weight}.Result)) {name=Weightmap Result, nodeColor=<rgba 0.5234 0 1 1>}
  NamedRerouteDeclaration(None=@n9.Previous Height) {name=Previous Height, nodeColor=<rgba 0 0.4609 1 1>}
  ScalarParameter() {defaultValue=0.2, parameterName=Patch Offset}
 ]
}
